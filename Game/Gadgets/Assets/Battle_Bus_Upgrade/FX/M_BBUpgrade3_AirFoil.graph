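# Unlit, two-sided translucent Niagara ribbon and sprite material, named as an airfoil effect for the Battle Bus upgrade gadget.
# Opacity multiplies particle alpha, a square edge mask fading along U, and camera-distance fade, then applies depth fade.
# Emissive color is particle color, optionally blended toward a fake chromatic split built from three offset edge masks packed into RGB.
# The chromatic blend defaults to off.
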Material /Game/Gadgets/Assets/Battle_Bus_Upgrade/FX/M_BBUpgrade3_AirFoil
graph {
 properties {
  bUsedWithNiagaraRibbons=true
  bUsedWithNiagaraSprites=true
  blendMode=BLEND_Translucent
  pixelDepthOffsetMode=PDOM_Legacy
  shadingModel=MSM_Unlit
  twoSided=true
  usedShadingModels=Unlit
 }
 nodes [
  n1 = ParticleColor()
  n2 = ComponentMask(None=TextureCoordinate()) {r=true}
  n3 = Multiply(
   A=Saturate(None=OneMinus(None=@n2))
   B=MaterialFunctionCall!(UVs, Mult=ScalarParameter() {defaultValue=4, parameterName=EdgeMult}, Exp=ScalarParameter() {defaultValue=2, parameterName=EdgeExp}) {materialFunction=/Game/Effects/Fort_Effects/MaterialFunctions/MF_SquareEdgeMask}.SquareMask
  )
  n4 = ScalarParameter() {defaultValue=75, parameterName=CamFade}
  DepthFade(
   Opacity=Multiply(
    A=Multiply(A=@n1.A, B=@n3)
    B=MaterialFunctionCall(Fade Length=@n4) {materialFunction=&CameraDepthFade}.Result
   )
   FadeDistance=ScalarParameter() {defaultValue=20, parameterName=DepthFadeDist}
  )
  n5 = TextureCoordinate()
  n6 = ComponentMask(None=VectorParameter() {parameterName=ChromaticOffset, defaultValue=<rgba 0.05 0.05 0 0>}.RGB) {g=true, r=true}
  n7 = ScalarParameter() {defaultValue=5, parameterName=Chroma_EdgeMult}
  n8 = ScalarParameter() {defaultValue=1, parameterName=Chroma_EdgeExp}
  n9 = Multiply(A=@n6) {constB=-1}
  n10 = MaterialFunctionCall(
   X=MaterialFunctionCall(UVs=Add(A=@n5, B=@n6), Mult=@n7, Exp=@n8) {materialFunction=&MF_SquareEdgeMask}.SquareMask
   Y=MaterialFunctionCall(UVs=@n5, Mult=@n7, Exp=@n8) {materialFunction=&MF_SquareEdgeMask}.SquareMask
   Z=MaterialFunctionCall(UVs=Add(A=@n5, B=@n9), Mult=@n7, Exp=@n8) {materialFunction=&MF_SquareEdgeMask}.SquareMask
  ) {materialFunction=&MakeFloat3}
  MaterialFunctionCall(Emissive=LinearInterpolate(A=@n1.RGB, B=Multiply(A=@n10.Result, B=@n1.RGB), Alpha=ScalarParameter() {parameterName=Use Chromatic})) {materialFunction=&MF_FX_EmissiveAdjustments}
 ]
}
