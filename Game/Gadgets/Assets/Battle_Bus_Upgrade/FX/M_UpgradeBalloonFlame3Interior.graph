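# Additive, two-sided unlit material for Niagara mesh particles, apparently the interior flame of the upgraded Battle Bus balloon.
# Two blended panning smoke-noise textures drive vertex displacement along normals and break up the V-gradient glow opacity, softened by facing angle and depth fade.
# Niagara dynamic parameters control tiling, pan speed, glow, pinch and vertex-red-weighted pull.
# Particle color, darkened by vertex red, feeds emissive; mobile skips HDR remapping and depth fade.
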Material /Game/Gadgets/Assets/Battle_Bus_Upgrade/FX/M_UpgradeBalloonFlame3Interior
graph {
 properties {
  bUsedWithNiagaraMeshParticles=true
  blendMode=BLEND_Additive
  pixelDepthOffsetMode=PDOM_Legacy
  shadingModel=MSM_Unlit
  twoSided=true
  usedShadingModels=Unlit
 }
 nodes [
  n1 = DynamicParameter() {parameterIndex=2, paramNames [YTile,PanSpeed,Param3,Param4]}
  n2 = AppendVector(A=ScalarParameter() {defaultValue=1, parameterName=X_Tile}, B=@n1.YTile)
  n3 = Multiply(A=Time(), B=@n1.PanSpeed)
  n4 = Panner(
   Coordinate=Multiply(A=@n2, B=TextureCoordinate() {uTiling=3, vTiling=0.5})
   Time=@n3
   Speed=AppendVector(A=ScalarParameter() {defaultValue=0.2, parameterName=NoiseA_SpeedX}, B=ScalarParameter() {defaultValue=1, parameterName=NoiseA_SpeedY})
  ) {speedX=0.2, speedY=1}
  n5 = Panner(
   Coordinate=Multiply(A=@n2, B=TextureCoordinate() {uTiling=3, vTiling=0.2})
   Time=@n3
   Speed=AppendVector(A=ScalarParameter() {defaultValue=-0.25, parameterName=NoiseB_SpeedX}, B=ScalarParameter() {defaultValue=0.8, parameterName=NoiseB_SpeedY})
  ) {speedX=-0.25, speedY=0.8}
  n6 = Multiply(A=LinearInterpolate(
    A=TextureSampleParameter2D(UVs=@n4) {parameterName=NoiseA, texture=&T_Inky_Smoke_Tile_Chromatic}.RGB
    B=TextureSampleParameter2D(UVs=@n5) {parameterName=NoiseB, texture=&T_Inky_Smoke_Tile_Chromatic}.RGB
   )) {constB=1.5}
  n7 = Multiply(A=VertexNormalWS(), B=@n6)
  n8 = ScalarParameter() {defaultValue=2, parameterName=WPO}
  n9 = DynamicParameter() {desc=Niagara index 1, parameterIndex=1, defaultValue=<rgba 0 1 0 1>, paramNames [WPOPinch,GlowMult,Pull,Param4]}
  n10 = TextureCoordinate()
  n11 = Multiply(A=Subtract(A=WorldPosition().XYZ, B=ParticlePositionWS()), B=Power(Base=ComponentMask(None=@n10) {g=true}) {constExponent=3})
  n12 = Transform(None=VectorParameter() {parameterName=Vector, defaultValue=<rgba 1 0 0 0>}.RGB) {transformSourceType=TRANSFORMSOURCE_Local}
  Add(
   A=Add(A=Multiply(A=@n7, B=@n8) {constB=7}, B=Multiply(A=@n9.WPOPinch, B=@n11))
   B=Multiply(A=Multiply(A=@n9.Pull, B=@n12), B=VertexColor().R)
  )
  n13 = OneMinus(None=VertexColor().R)
  n14 = ParticleColor()
  n15 = Clamp(None=Multiply(A=Power(Base=@n13), B=@n14.RGB))
  n16 = Multiply(A=@n14.RGB, B=ScalarParameter() {defaultValue=0.75, parameterName=HDRMult})
  MaterialFunctionCall(Emissive=ShadingPathSwitch(
    Default=MaterialFunctionCall(LDR Input=@n15, HDR Tint=@n16) {materialFunction=&DeriveHDRfromLDR}.RemappedRange
    Mobile=@n15
   ) {desc=Migrated from FeatureLevelSwitch}) {materialFunction=&MF_FX_EmissiveAdjustments}
  n17 = ParticleColor()
  n18 = DynamicParameter() {desc=Niagara index 1, parameterIndex=1, defaultValue=<rgba 0 1 0 1>, paramNames [WPOPinch,GlowMult,Pull,Param4]}
  n19 = ComponentMask(None=TextureCoordinate()) {g=true}
  n20 = ScalarParameter() {defaultValue=6, parameterName=GlowPower}
  n21 = Multiply(
   A=@n18.GlowMult
   B=Multiply(A=ScalarParameter() {defaultValue=0.5, parameterName=GlowBrightness}, B=Power(Base=@n19, Exp=@n20) {constExponent=6}) {constB=0.5}
  )
  n22 = ScalarParameter() {defaultValue=2, parameterName=NoiseFadePower}
  n23 = Add(A=@n21, B=Multiply(A=Power(Base=@n19, Exp=@n22), B=Add(A=@n21, B=@n6)))
  n24 = DotProduct(A=CameraVectorWS(), B=VertexNormalWS())
  n25 = Multiply(A=Power(Base=Abs(None=@n24), Exp=ScalarParameter() {defaultValue=4, parameterName=DirectionalFadePower}) {constExponent=4}) {constB=4}
  n26 = ScalarParameter() {defaultValue=50, parameterName=DepthFadeDist}
  Multiply(
   A=Multiply(A=Multiply(A=@n17.A, B=@n23), B=Clamp(None=@n25))
   B=ShadingPathSwitch(Default=DepthFade(FadeDistance=@n26), Mobile=Constant() {r=1})
  )
 ]
}
